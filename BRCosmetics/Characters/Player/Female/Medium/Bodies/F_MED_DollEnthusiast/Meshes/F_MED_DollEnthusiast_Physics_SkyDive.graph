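# Physics asset for the skydive variant of the F_MED_DollEnthusiast player body, set up for dynamic jacket bones.
# Five jacket chains (back, left, left-back, right, right-back) each hang three tightly limited capsules from a master bone tied to the pelvis with free rotation.
# Free-rotation constraints also link front and back tips on each side.
# Body capsules cover pelvis, lower spine and thighs, with calf and foot on the right only.
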PhysicsAsset /BRCosmetics/Characters/Player/Female/Medium/Bodies/F_MED_DollEnthusiast/Meshes/F_MED_DollEnthusiast_Physics_SkyDive
bodies {
 calf_r [{shapeType=Capsule, length=30, radius=12, shapeName=calf_r_capsule, center=<25.4057 5.7781 -2.3151>, rotation=<rot 0.0619 -82.1036 -93.0609>}]
 dyn_jacket_b_01 [{shapeType=Capsule, length=5, radius=3, shapeName=dyn_jacket_b_01_capsule, center=<5 0 0>}]
 dyn_jacket_b_01_MSTR [{shapeType=Capsule, length=0.505, radius=0.505, shapeName=dyn_jacket_b_01_MSTR_capsule, rotation=<rot 0 -90 0>}]
 dyn_jacket_b_02 [{shapeType=Capsule, length=10, radius=6.8, shapeName=dyn_jacket_b_02_capsule, center=<8.203 0 0>}]
 dyn_jacket_b_03 [{shapeType=Capsule, length=14, radius=7.5, shapeName=dyn_jacket_b_03_capsule, center=<5 0 0>}]
 dyn_jacket_l_01 [{shapeType=Capsule, length=5, radius=3, shapeName=dyn_jacket_l_01_capsule, center=<5 0 0>, rotation=<rot 0 0 90>}]
 dyn_jacket_l_01_MSTR [{shapeType=Capsule, length=0.505, radius=0.505, shapeName=dyn_jacket_l_01_MSTR_capsule, rotation=<rot 0 -90 0>}]
 dyn_jacket_l_02 [{shapeType=Capsule, length=7.433, radius=3.5, shapeName=dyn_jacket_l_02_capsule, center=<5 0 0>, rotation=<rot 0 0 90>}]
 dyn_jacket_l_03 [{shapeType=Capsule, length=11, radius=4.7, shapeName=dyn_jacket_l_03_capsule, center=<5 0 0>, rotation=<rot -0.0132 0.0639 104.8711>}]
 dyn_jacket_l_b_01 [{shapeType=Capsule, length=5, radius=3, shapeName=dyn_jacket_l_b_01_capsule, center=<5 0 0>, rotation=<rot 0 0 90>}]
 dyn_jacket_l_b_01_MSTR [{shapeType=Capsule, length=0.505, radius=0.505, shapeName=dyn_jacket_l_b_01_MSTR_capsule, rotation=<rot 0 -90 0>}]
 dyn_jacket_l_b_02 [{shapeType=Capsule, length=7.0428, radius=4, shapeName=dyn_jacket_l_b_02_capsule, center=<5 0 0>, rotation=<rot 0 0 90>}]
 dyn_jacket_l_b_03 [{shapeType=Capsule, length=10.6793, radius=3.2, shapeName=dyn_jacket_l_b_03_capsule, center=<5 0 0>, rotation=<rot -0.003 0.0041 73.6661>}]
 dyn_jacket_r_01 [{shapeType=Capsule, length=5, radius=3, shapeName=dyn_jacket_r_01_capsule, center=<5 0 0>, rotation=<rot 0 0 90>}]
 dyn_jacket_r_01_MSTR [{shapeType=Capsule, length=0.505, radius=0.505, shapeName=dyn_jacket_r_01_MSTR_capsule, rotation=<rot 0 -90 0>}]
 dyn_jacket_r_02 [{shapeType=Capsule, length=7.5182, radius=3.5, shapeName=dyn_jacket_r_02_capsule, center=<5 0 1.3072>, rotation=<rot 0 0 90>}]
 dyn_jacket_r_03 [{shapeType=Capsule, length=11, radius=5.5, shapeName=dyn_jacket_r_03_capsule, center=<5 0 0>, rotation=<rot 0 0 90>}]
 dyn_jacket_r_b_01 [{shapeType=Capsule, length=5, radius=3, shapeName=dyn_jacket_r_b_01_capsule, center=<5 0 0>, rotation=<rot 0 0 90>}]
 dyn_jacket_r_b_01_MSTR [{shapeType=Capsule, length=0.505, radius=0.505, shapeName=dyn_jacket_r_b_01_MSTR_capsule, rotation=<rot 0 -90 0>}]
 dyn_jacket_r_b_02 [{
   shapeType=Capsule
   length=12
   radius=6.5
   shapeName=dyn_jacket_r_b_02_capsule
   center=<10.0244 1.158 4.1377>
   rotation=<rot -0.0499 -0.0675 108.966>
  }]
 dyn_jacket_r_b_03 [{
   shapeType=Capsule
   length=20
   radius=7
   shapeName=dyn_jacket_l_b_03_capsule
   center=<10.7943 0.4449 1.1399>
   rotation=<rot 0 -0.0001 119.0558>
  }]
 foot_r [{shapeType=Capsule, length=15.7986, radius=10, shapeName=foot_r_capsule, center=<2.0109 -3.5875 0.1475>, rotation=<rot 0 4.6049 -86.3058>}]
 pelvis [{shapeType=Capsule, length=9, radius=10, shapeName=pelvis_capsule, center=<0 2 0>}]
 spine_01 [{shapeType=Capsule, length=0.505, radius=0.505, shapeName=spine_01_capsule, rotation=<rot 0 -90 0>}]
 spine_02 [{shapeType=Capsule, length=8, radius=7.5, shapeName=spine_02_capsule, center=<0 2 0>}]
 thigh_l [{shapeType=Capsule, length=35, radius=8, shapeName=thigh_l_capsule, center=<-25 0 0>, rotation=<rot 90 0 0>}]
 thigh_r [{shapeType=Capsule, length=40, radius=8.5, shapeName=thigh_l_capsule, center=<25 0 0>, rotation=<rot 90 180 0>}]
}
constraints [
 {
  bone1Name=thigh_l
  bone2Name=pelvis
  swing1LimitDegrees=45
  swing1Motion=Free
  swing2LimitDegrees=45
  swing2Motion=Free
  twistLimitDegrees=45
  twistMotion=Free
 }
 {
  bone1Name=spine_01
  bone2Name=pelvis
  swing1LimitDegrees=45
  swing1Motion=Free
  swing2LimitDegrees=45
  swing2Motion=Free
  twistLimitDegrees=45
  twistMotion=Free
 }
 {
  bone1Name=spine_02
  bone2Name=spine_01
  swing1LimitDegrees=45
  swing1Motion=Free
  swing2LimitDegrees=45
  swing2Motion=Free
  twistLimitDegrees=45
  twistMotion=Free
 }
 {
  bone1Name=thigh_r
  bone2Name=pelvis
  swing1LimitDegrees=45
  swing1Motion=Free
  swing2LimitDegrees=45
  swing2Motion=Free
  twistLimitDegrees=45
  twistMotion=Free
 }
 {bone1Name=dyn_jacket_b_01, bone2Name=dyn_jacket_b_01_MSTR, swing1LimitDegrees=10, swing2LimitDegrees=5, twistLimitDegrees=2}
 {bone1Name=dyn_jacket_b_02, bone2Name=dyn_jacket_b_01, swing1LimitDegrees=10, swing2LimitDegrees=5, twistLimitDegrees=2}
 {bone1Name=dyn_jacket_b_03, bone2Name=dyn_jacket_b_02, swing1LimitDegrees=10, swing2LimitDegrees=5, twistLimitDegrees=2}
 {bone1Name=dyn_jacket_l_01, bone2Name=dyn_jacket_l_01_MSTR, swing1LimitDegrees=5, swing2LimitDegrees=10, twistLimitDegrees=2}
 {bone1Name=dyn_jacket_l_02, bone2Name=dyn_jacket_l_01, swing1LimitDegrees=5, swing2LimitDegrees=10, twistLimitDegrees=2}
 {bone1Name=dyn_jacket_l_03, bone2Name=dyn_jacket_l_02, swing1LimitDegrees=5, swing2LimitDegrees=10, twistLimitDegrees=2}
 {bone1Name=dyn_jacket_l_b_01, bone2Name=dyn_jacket_l_b_01_MSTR, swing1LimitDegrees=5, swing2LimitDegrees=10, twistLimitDegrees=2}
 {bone1Name=dyn_jacket_l_b_02, bone2Name=dyn_jacket_l_b_01, swing1LimitDegrees=5, swing2LimitDegrees=10, twistLimitDegrees=2}
 {bone1Name=dyn_jacket_l_b_03, bone2Name=dyn_jacket_l_b_02, swing1LimitDegrees=5, swing2LimitDegrees=10, twistLimitDegrees=2}
 {bone1Name=dyn_jacket_r_01, bone2Name=dyn_jacket_r_01_MSTR, swing1LimitDegrees=5, swing2LimitDegrees=10, twistLimitDegrees=2}
 {bone1Name=dyn_jacket_r_02, bone2Name=dyn_jacket_r_01, swing1LimitDegrees=5, swing2LimitDegrees=10, twistLimitDegrees=2}
 {bone1Name=dyn_jacket_r_03, bone2Name=dyn_jacket_r_02, swing1LimitDegrees=5, swing2LimitDegrees=10, twistLimitDegrees=2}
 {bone1Name=dyn_jacket_r_b_01, bone2Name=dyn_jacket_r_b_01_MSTR, swing1LimitDegrees=5, swing2LimitDegrees=10, twistLimitDegrees=2}
 {bone1Name=dyn_jacket_r_b_02, bone2Name=dyn_jacket_r_b_01, swing1LimitDegrees=5, swing2LimitDegrees=10, twistLimitDegrees=2}
 {
  bone1Name=dyn_jacket_r_b_01_MSTR
  bone2Name=pelvis
  swing1LimitDegrees=45
  swing1Motion=Free
  swing2LimitDegrees=45
  swing2Motion=Free
  twistLimitDegrees=45
  twistMotion=Free
 }
 {
  bone1Name=dyn_jacket_r_01_MSTR
  bone2Name=pelvis
  swing1LimitDegrees=45
  swing1Motion=Free
  swing2LimitDegrees=45
  swing2Motion=Free
  twistLimitDegrees=45
  twistMotion=Free
 }
 {
  bone1Name=dyn_jacket_l_b_01_MSTR
  bone2Name=pelvis
  swing1LimitDegrees=45
  swing1Motion=Free
  swing2LimitDegrees=45
  swing2Motion=Free
  twistLimitDegrees=45
  twistMotion=Free
 }
 {
  bone1Name=dyn_jacket_l_01_MSTR
  bone2Name=pelvis
  swing1LimitDegrees=45
  swing1Motion=Free
  swing2LimitDegrees=45
  swing2Motion=Free
  twistLimitDegrees=45
  twistMotion=Free
 }
 {
  bone1Name=dyn_jacket_b_01_MSTR
  bone2Name=pelvis
  swing1LimitDegrees=45
  swing1Motion=Free
  swing2LimitDegrees=45
  swing2Motion=Free
  twistLimitDegrees=45
  twistMotion=Free
 }
 {bone1Name=dyn_jacket_r_b_03, bone2Name=dyn_jacket_r_b_02, swing1LimitDegrees=5, swing2LimitDegrees=10, twistLimitDegrees=2}
 {
  bone1Name=foot_r
  bone2Name=calf_r
  swing1LimitDegrees=45
  swing1Motion=Free
  swing2LimitDegrees=45
  swing2Motion=Free
  twistLimitDegrees=45
  twistMotion=Free
 }
 {
  bone1Name=calf_r
  bone2Name=thigh_r
  swing1LimitDegrees=45
  swing1Motion=Free
  swing2LimitDegrees=45
  swing2Motion=Free
  twistLimitDegrees=45
  twistMotion=Free
 }
 {
  bone1Name=dyn_jacket_l_03
  bone2Name=dyn_jacket_l_b_03
  swing1LimitDegrees=45
  swing1Motion=Free
  swing2LimitDegrees=45
  swing2Motion=Free
  twistLimitDegrees=45
  twistMotion=Free
 }
 {
  bone1Name=dyn_jacket_r_03
  bone2Name=dyn_jacket_r_b_03
  swing1LimitDegrees=45
  swing1Motion=Free
  swing2LimitDegrees=45
  swing2Motion=Free
  twistLimitDegrees=45
  twistMotion=Free
 }
]
