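# Post-animation AnimBP for the F_MED_DollEnthusiast cosmetic body, adding secondary motion to its jacket and llama-head scarf.
# The copied parent pose feeds a state-driven Control Rig, then one of two RigidBody setups, switching for skydive floating or parachuting.
# AnimDynamics then simulates the scarf except while floating, and additive bone offsets correct jacket and scarf per movement state.
# The EventGraph derives those states, smoothing some, and computes a gravity override.
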AnimBlueprint /BRCosmetics/Characters/Player/Female/Medium/Bodies/F_MED_DollEnthusiast/Meshes/F_MED_DollEnthusiast_AnimBP
target_skeleton=&/F_MED_DollEnthusiast_Skeleton
graphs [
 AnimGraph {nodes [
   n1 = "Constraint - dyn_scarf_01 (PPPPPPPP)"!(
    ComponentPose
    Alpha=1
    ConstraintWeights_0=0.0004
    ConstraintWeights_1:float=0.4994
    ConstraintWeights_2:float=0.0009
    ConstraintWeights_3:float=0.182
    ConstraintWeights_4:float=0.0002
    ConstraintWeights_5:float=0.2087
    ConstraintWeights_6:float=0.0003
    ConstraintWeights_7:float=0.1081
   ) -> (Pose) {
    alphaBoolBlend {blendOption=Linear}
    boneToModify {boneName=dyn_scarf_01}
    constraintSetup [
     {targetBone {boneName=spine_02}}
     {targetBone {boneName=spine_04}}
     {targetBone {boneName=spine_03}}
     {targetBone {boneName=spine_05}}
     {targetBone {boneName=clavicle_l}}
     {targetBone {boneName=clavicle_r}}
     {targetBone {boneName=pelvis}}
     {targetBone {boneName=upperarm_twist_01_r}}
    ]
    constraintWeights [1,1,1,1,1,1,1,1]
   }
   "Constraint - dyn_scarf_lamahead_head (PPPPPPPP)"!(
    ComponentPose=@n1
    Alpha=1
    ConstraintWeights_0=0.0004
    ConstraintWeights_1:float=0.4994
    ConstraintWeights_2:float=0.0009
    ConstraintWeights_3:float=0.182
    ConstraintWeights_4:float=0.0002
    ConstraintWeights_5:float=0.2087
    ConstraintWeights_6:float=0.0003
    ConstraintWeights_7:float=0.1081
   ) -> (Pose) {
    alphaBoolBlend {blendOption=Linear}
    boneToModify {boneName=dyn_scarf_lamahead_head}
    constraintSetup [
     {targetBone {boneName=spine_02}}
     {targetBone {boneName=spine_04}}
     {targetBone {boneName=spine_03}}
     {targetBone {boneName=spine_05}}
     {targetBone {boneName=clavicle_l}}
     {targetBone {boneName=clavicle_r}}
     {targetBone {boneName=pelvis}}
     {targetBone {boneName=upperarm_twist_01_r}}
    ]
    constraintWeights [1,1,1,1,1,1,1,1]
   }
   "Transform (Modify) Bone - Bone: dyn_jacket_r_01_MSTR"!(ComponentPose, bAlphaBoolEnabled:bool=Get bIsCrouching(), Translation="0,0,0", Rotation="0,0,0", Scale="1,1,1") -> (Pose) {
    alphaInputType=Bool
    rotationMode=BMM_Additive
    rotationSpace=BCS_BoneSpace
    translationMode=BMM_Additive
    translationSpace=BCS_BoneSpace
    alphaBoolBlend {blendInTime=0.2, blendOption=Linear, blendOutTime=0.2}
    alphaScaleBiasClamp {interpSpeedIncreasing=0}
    boneToModify {boneName=dyn_jacket_r_01_MSTR}
   }
   n2 = Get jumping and falling() -> (jumping and falling:bool)
   n3 = "Transform (Modify) Bone - Bone: dyn_jacket_r_01_MSTR"(
    ComponentPose=Local To Component(LocalPose=Use cached pose 'scarf'())
    bAlphaBoolEnabled=@n2
    Translation="0,0,0"
    Rotation="12.368,15.419,-3.659"
    Scale="1,1,1"
   ) {
    alphaInputType=Bool
    rotationMode=BMM_Additive
    rotationSpace=BCS_BoneSpace
    translationMode=BMM_Additive
    translationSpace=BCS_BoneSpace
    alphaBoolBlend {blendInTime=0.2, blendOption=Linear, blendOutTime=0.3}
    boneToModify {boneName=dyn_jacket_r_01_MSTR}
    rotation=<rot 12.368 15.419 -3.659>
   }
   n4 = "Transform (Modify) Bone - Bone: dyn_jacket_r_b_01_MSTR"!(ComponentPose=@n3, bAlphaBoolEnabled:bool=@n2, Translation="0,0,0", Rotation="-5.546,-3.243,2.299", Scale="1,1,1") -> (Pose) {
    alphaInputType=Bool
    rotationMode=BMM_Additive
    rotationSpace=BCS_BoneSpace
    translationMode=BMM_Additive
    translationSpace=BCS_BoneSpace
    alphaBoolBlend {blendInTime=0.2, blendOption=Linear, blendOutTime=0.3}
    boneToModify {boneName=dyn_jacket_r_b_01_MSTR}
    rotation=<rot -5.546 -3.243 2.299>
   }
   n5 = "Transform (Modify) Bone - Bone: dyn_jacket_r_02"!(ComponentPose=@n4, bAlphaBoolEnabled:bool=@n2, Translation="0,0,0", Rotation="5.3062,0,0", Scale="1,1,1") -> (Pose) {
    alphaInputType=Bool
    rotationMode=BMM_Additive
    rotationSpace=BCS_BoneSpace
    translationMode=BMM_Additive
    translationSpace=BCS_BoneSpace
    alphaBoolBlend {blendInTime=0.2, blendOption=Linear, blendOutTime=0.3}
    boneToModify {boneName=dyn_jacket_r_02}
    rotation {pitch=5.3062}
   }
   n6 = Get jumping and falling()
   n7 = "Transform (Modify) Bone - Bone: dyn_jacket_r_01_MSTR"(ComponentPose=@n5, bAlphaBoolEnabled=@n6, Translation="0,0,0", Rotation="-5.629,-8.174,1.088", Scale="1,1,1") {
    alphaInputType=Bool
    rotationMode=BMM_Additive
    rotationSpace=BCS_BoneSpace
    translationMode=BMM_Additive
    translationSpace=BCS_BoneSpace
    alphaBoolBlend {blendInTime=0.2, blendOption=Linear, blendOutTime=0.2}
    boneToModify {boneName=dyn_jacket_r_01_MSTR}
    rotation=<rot -5.629 -8.174 1.088>
   }
   n8 = "Transform (Modify) Bone - Bone: dyn_jacket_r_b_01_MSTR"(ComponentPose=@n7, bAlphaBoolEnabled=@n6, Translation="0,0,0", Rotation="-1.628,2.185,3.726", Scale="1,1,1") {
    alphaInputType=Bool
    rotationMode=BMM_Additive
    rotationSpace=BCS_BoneSpace
    translationMode=BMM_Additive
    translationSpace=BCS_BoneSpace
    alphaBoolBlend {blendInTime=0.2, blendOption=Linear, blendOutTime=0.2}
    boneToModify {boneName=dyn_jacket_r_b_01_MSTR}
    rotation=<rot -1.628 2.185 3.726>
   }
   n9 = "Transform (Modify) Bone - Bone: dyn_jacket_r_b_01_MSTR"(ComponentPose=@n8, bAlphaBoolEnabled=Get bIsSkydiveDiveMode(), Translation="0,0,0", Rotation="12.175,-0.000,-3.667", Scale="1,1,1") {
    alphaInputType=Bool
    rotationMode=BMM_Additive
    rotationSpace=BCS_BoneSpace
    translationMode=BMM_Additive
    translationSpace=BCS_BoneSpace
    alphaBoolBlend {blendInTime=0.2, blendOption=Linear, blendOutTime=0.2}
    boneToModify {boneName=dyn_jacket_r_b_01_MSTR}
    rotation=<rot 12.175 0 -3.667>
   }
   n10 = "Transform (Modify) Bone - Bone: dyn_jacket_b_03"!(ComponentPose=@n9, bAlphaBoolEnabled:bool=Get bIsBaseballSliding(), Translation="0,0,0", Rotation="6.023,72.685,1.322", Scale="1,1,1") -> (Pose) {
    alphaInputType=Bool
    rotationMode=BMM_Additive
    rotationSpace=BCS_BoneSpace
    translationMode=BMM_Additive
    translationSpace=BCS_BoneSpace
    alphaBoolBlend {blendInTime=0.2, blendOption=Linear, blendOutTime=0.2}
    boneToModify {boneName=dyn_jacket_b_03}
    rotation=<rot 6.023 72.685 1.322>
   }
   n11 = "Transform (Modify) Bone - Bone: dyn_jacket_b_03"(ComponentPose=@n10, bAlphaBoolEnabled=Get bIsKneeSliding(), Translation="0,0,0", Rotation="6.080,50.772,-1.018", Scale="1,1,1") {
    alphaInputType=Bool
    rotationMode=BMM_Additive
    rotationSpace=BCS_BoneSpace
    translationMode=BMM_Additive
    translationSpace=BCS_BoneSpace
    alphaBoolBlend {blendInTime=0.2, blendOption=Linear, blendOutTime=0.2}
    boneToModify {boneName=dyn_jacket_b_03}
    rotation=<rot 6.08 50.772 -1.018>
   }
   n12 = "Transform (Modify) Bone - Bone: dyn_scarf_lamahead_head"!(ComponentPose=@n11, bAlphaBoolEnabled:bool=Get bIsParachuteOpened(), Translation="-0.3,0,0", Rotation="-0.323,-8.630,-5.980", Scale="1,1,1") -> (Pose) {
    alphaInputType=Bool
    rotationMode=BMM_Additive
    rotationSpace=BCS_BoneSpace
    translationMode=BMM_Additive
    translationSpace=BCS_BoneSpace
    alphaBoolBlend {blendInTime=0.2, blendOption=Linear, blendOutTime=0.2}
    boneToModify {boneName=dyn_scarf_lamahead_head}
    rotation=<rot -0.323 -8.63 -5.98>
    translation {x=-0.3}
   }
   n13 = Get crouching not moving() -> (crouching not moving:double)
   n14 = "Transform (Modify) Bone - Bone: dyn_jacket_r_b_01_MSTR"!(ComponentPose=@n12, Alpha:float=@n13, Translation="0,0,0", Rotation="-0.000,3.643,-2.339", Scale="1,1,1") -> (Pose) {
    rotationMode=BMM_Additive
    rotationSpace=BCS_BoneSpace
    translationMode=BMM_Additive
    translationSpace=BCS_BoneSpace
    alphaBoolBlend {blendOption=Linear, blendOutTime=0.2}
    alphaScaleBiasClamp {bInterpResult=true, interpSpeedDecreasing=20}
    boneToModify {boneName=dyn_jacket_r_b_01_MSTR}
    rotation=<rot 0 3.643 -2.339>
   }
   n15 = "Transform (Modify) Bone - Bone: dyn_jacket_b_01_MSTR"(ComponentPose=@n14, Alpha=@n13, Translation="0,0,0", Rotation="0.054,-10.187,1.137", Scale="1,1,1") {
    rotationMode=BMM_Additive
    rotationSpace=BCS_BoneSpace
    translationMode=BMM_Additive
    translationSpace=BCS_BoneSpace
    alphaBoolBlend {blendOption=Linear, blendOutTime=0.2}
    alphaScaleBiasClamp {bInterpResult=true, interpSpeedDecreasing=20}
    boneToModify {boneName=dyn_jacket_b_01_MSTR}
    rotation=<rot 0.054 -10.1869 1.137>
   }
   n16 = "Transform (Modify) Bone - Bone: dyn_jacket_l_b_01_MSTR"(ComponentPose=@n15, Alpha=@n13, Translation="0,0,0", Rotation="9.578,0.000,-0.000", Scale="1,1,1") {
    rotationMode=BMM_Additive
    rotationSpace=BCS_BoneSpace
    translationMode=BMM_Additive
    translationSpace=BCS_BoneSpace
    alphaBoolBlend {blendInTime=0.2, blendOption=Linear, blendOutTime=0.2}
    alphaScaleBiasClamp {bInterpResult=true, interpSpeedDecreasing=20}
    boneToModify {boneName=dyn_jacket_l_b_01_MSTR}
    rotation=<rot 9.578 0 0>
   }
   n17 = Get bIsSkydiveFloating()
   n18 = "Transform (Modify) Bone - Bone: dyn_scarf_01"!(ComponentPose=@n16, bAlphaBoolEnabled:bool=@n17, Translation="-2,0,0", Rotation="4.419,6.448,7.917", Scale="1,1,1") -> (Pose) {
    alphaInputType=Bool
    rotationMode=BMM_Additive
    rotationSpace=BCS_BoneSpace
    translationMode=BMM_Additive
    translationSpace=BCS_BoneSpace
    alphaBoolBlend {blendInTime=0.2, blendOption=Linear, blendOutTime=0.2}
    boneToModify {boneName=dyn_scarf_01}
    rotation=<rot 4.419 6.448 7.917>
    translation {x=-2}
   }
   n19 = "Transform (Modify) Bone - Bone: dyn_scarf_03"!(ComponentPose=@n18, bAlphaBoolEnabled:bool=@n17, Translation="0,0,0", Rotation="0.624,-2.673,16.050", Scale="1,1,1") -> (Pose) {
    alphaInputType=Bool
    rotationMode=BMM_Additive
    rotationSpace=BCS_BoneSpace
    translationMode=BMM_Additive
    translationSpace=BCS_BoneSpace
    alphaBoolBlend {blendInTime=0.2, blendOption=Linear, blendOutTime=0.2}
    boneToModify {boneName=dyn_scarf_03}
    rotation=<rot 0.624 -2.673 16.0499>
   }
   n20 = "Transform (Modify) Bone - Bone: dyn_scarf_01"(ComponentPose=@n19, bAlphaBoolEnabled=Get bIsParachuteOpened(), Translation="-2,0,0", Rotation="2.251,0.000,0.000", Scale="1,1,1") {
    alphaInputType=Bool
    rotationMode=BMM_Additive
    rotationSpace=BCS_BoneSpace
    translationMode=BMM_Additive
    translationSpace=BCS_BoneSpace
    alphaBoolBlend {blendInTime=0.2, blendOption=Linear, blendOutTime=0.2}
    boneToModify {boneName=dyn_scarf_01}
    rotation=<rot 2.251 0 0>
    translation {x=-2}
   }
   n21 = Get bIsOnGrindRail()
   n22 = "Transform (Modify) Bone - Bone: dyn_jacket_r_01_MSTR"(ComponentPose=@n20, bAlphaBoolEnabled=@n21, Translation="0,0,0", Rotation="4.520,27.202,0.003", Scale="1,1,1") {
    alphaInputType=Bool
    rotationMode=BMM_Additive
    rotationSpace=BCS_BoneSpace
    translationMode=BMM_Additive
    translationSpace=BCS_BoneSpace
    alphaBoolBlend {blendInTime=0.2, blendOption=Linear, blendOutTime=0.2}
    boneToModify {boneName=dyn_jacket_r_01_MSTR}
    rotation=<rot 4.52 27.202 0.003>
   }
   n23 = "Transform (Modify) Bone - Bone: dyn_jacket_r_b_01_MSTR"!(ComponentPose=@n22, bAlphaBoolEnabled:bool=@n21, Translation="0,0,0", Rotation="11.754,23.268,10.992", Scale="1,1,1") -> (Pose) {
    alphaInputType=Bool
    rotationMode=BMM_Additive
    rotationSpace=BCS_BoneSpace
    translationMode=BMM_Additive
    translationSpace=BCS_BoneSpace
    alphaBoolBlend {blendInTime=0.2, blendOption=Linear, blendOutTime=0.2}
    boneToModify {boneName=dyn_jacket_r_b_01_MSTR}
    rotation=<rot 11.754 23.268 10.992>
   }
   Output Pose(Result=Component To Local(ComponentPose=@n23)) {layerGroup=None}
   n24 = RigidBody!(ComponentPose=Local To Component(LocalPose=Use cached pose 'ControlRig'()), Alpha, OverrideWorldGravity:Vector="0,0,-980", ExternalForce) -> (Pose) {
    bClampLinearTranslationLimitToRefPose=true
    bOverrideWorldGravity=true
    overridePhysicsAsset=&/F_MED_DollEnthusiast_Physics_SkyDive
    simulationSpace=BaseBoneSpace
    alphaBoolBlend {blendOption=Linear}
    baseBoneRef {boneName=pelvis}
    componentAppliedLinearAccClamp=<1000 1000 1000>
    componentLinearAccScale=<0.1 0.1 0.1>
    componentLinearVelScale=<0.2 0.2 0.2>
    overrideWorldGravity {z=-980}
    simSpaceSettings {maxAngularAcceleration=1000, maxAngularVelocity=3, maxLinearAcceleration=250, maxLinearVelocity=500, velocityScaleZ=0}
   }
   n25 = RigidBody(ComponentPose=Local To Component(LocalPose=Use cached pose 'ControlRig'()), OverrideWorldGravity=Get Gravity Override()) {
    bClampLinearTranslationLimitToRefPose=true
    bOverrideWorldGravity=true
    overridePhysicsAsset=&/F_MED_DollEnthusiast_Physics
    simulationSpace=BaseBoneSpace
    alphaBoolBlend {blendOption=Linear}
    baseBoneRef {boneName=pelvis}
    componentAppliedLinearAccClamp=<1000 2500 2500>
    componentLinearAccScale=<0.3 0.3 0.1>
    componentLinearVelScale=<0.4 0.3 0.2>
    overrideWorldGravity {z=-980}
    simSpaceSettings {maxAngularAcceleration=1000, maxAngularVelocity=3, maxLinearAcceleration=250, maxLinearVelocity=500, velocityScaleZ=0, worldAlpha=0.2}
   }
   Save cached pose 'RB'(Pose=Blend Poses by bool(
     BlendPose_0=Component To Local(ComponentPose=@n24)
     BlendPose_1=Component To Local(ComponentPose=@n25)
     bActiveValue=Get turn off gravity override() -> (turn off gravity override:bool)
    ) {blendType=Linear}) {cacheName=RB}
   n26 = Control Rig!(
    Source=Copy Pose From Mesh() {bUseAttachedParent=true}
    Alpha
    sky diving:double=Get sky diving() -> (sky diving:double)
    is crouching:double=Get crouching float() -> (crouching float:double)
    is crouching moving:double=Get crouching and moving 2d() -> (crouching and moving 2d:double)
    crouching not moving:double=Get crouching not moving()
    is knee sliding:double=Get is knee sliding() -> (is knee sliding:double)
    is baseball sliding:double=Get is baseball sliding() -> (is baseball sliding:double)
    playing melee:double=Get playing melee() -> (playing melee:double)
    is playing emote:double=Get is playing emote() -> (is playing emote:double)
    is in frontend:double=Get is in frontend() -> (is in frontend:double)
    melee fix:double=Get melee fix() -> (melee fix:double)
   ) -> (Pose) {
    alphaBoolBlend {blendOption=Linear}
    controlRigAssetReference {blueprintRigClass=&/F_MED_DollEnthusiast_CtrlRig.F_MED_DollEnthusiast_CtrlRig_C}
   }
   Save cached pose 'ControlRig'(Pose=@n26) {cacheName=ControlRig}
   n27 = Get scarf is on() -> (scarf is on:bool)
   n28 = Anim Dynamics - Bone: dyn_scarf_01!(ComponentPose=Local To Component(LocalPose=Use cached pose 'RB'()), bAlphaBoolEnabled:bool=@n27) -> (Pose) {
    alphaInputType=Bool
    angularBiasOverride=1
    angularDampingOverride=0.999
    angularSpringConstant=1500
    bGravityOverrideInSimSpace=true
    bOverrideAngularDamping=true
    bUseGravityOverride=true
    bUseSphericalLimits=true
    gravityScale=0.35
    simulationSpace=RootRelative
    alphaBoolBlend {blendInTime=0.1, blendOption=Linear, blendOutTime=0.2}
    boundBone {boneName=dyn_scarf_01}
    componentAppliedLinearAccClamp=<1500 1500 1500>
    componentLinearAccScale=<0.1 0.1 0.1>
    componentLinearVelScale=<0.1 0.1 0.1>
    gravityOverride {z=-980}
    physicsBodyDefinitions [{
      collisionType=CustomSphere
      sphereCollisionRadius=5
      boundBone {boneName=dyn_scarf_01}
      boxExtents=<30 30 30>
      constraintSetup {angularLimitsMax=<0 7 20>, angularLimitsMin=<0 -10 0>, angularTarget=<0 0 0>}
      localJointOffset=<5 0 0>
     }]
   }
   n29 = Anim Dynamics - Bone: dyn_scarf_02!(ComponentPose=@n28, bAlphaBoolEnabled:bool=@n27) -> (Pose) {
    alphaInputType=Bool
    angularBiasOverride=1
    angularDampingOverride=0.999
    angularSpringConstant=1500
    bGravityOverrideInSimSpace=true
    bOverrideAngularDamping=true
    bUseGravityOverride=true
    gravityScale=0.35
    simulationSpace=RootRelative
    alphaBoolBlend {blendInTime=0.1, blendOption=Linear, blendOutTime=0.2}
    boundBone {boneName=dyn_scarf_02}
    componentAppliedLinearAccClamp=<1500 1500 1500>
    componentLinearAccScale=<0.1 0.1 0.1>
    componentLinearVelScale=<0.1 0.1 0.1>
    gravityOverride {z=-980}
    physicsBodyDefinitions [{
      collisionType=CustomSphere
      sphereCollisionRadius=5
      boundBone {boneName=dyn_scarf_02}
      boxExtents=<25 25 25>
      constraintSetup {angularLimitsMax=<0 7 20>, angularLimitsMin=<0 -10 0>, angularTarget=<0 0 0>}
      localJointOffset=<5 0 0>
     }]
    sphericalLimits [{limitRadius=5, drivingBone {boneName=spine_05}, sphereLocalOffset=<-8.4976 10 12.8023>}]
   }
   n30 = Anim Dynamics - Bone: dyn_scarf_03!(ComponentPose=@n29, bAlphaBoolEnabled:bool=@n27) -> (Pose) {
    alphaInputType=Bool
    angularBiasOverride=1
    angularDampingOverride=0.999
    angularSpringConstant=1500
    bGravityOverrideInSimSpace=true
    bOverrideAngularDamping=true
    bUseGravityOverride=true
    gravityScale=0.35
    simulationSpace=RootRelative
    alphaBoolBlend {blendInTime=0.1, blendOption=Linear, blendOutTime=0.2}
    boundBone {boneName=dyn_scarf_03}
    componentAppliedLinearAccClamp=<1500 1500 1500>
    componentLinearAccScale=<0.1 0.1 0.1>
    componentLinearVelScale=<0.1 0.1 0.1>
    gravityOverride {z=-980}
    physicsBodyDefinitions [{
      collisionType=CustomSphere
      sphereCollisionRadius=5
      boundBone {boneName=dyn_scarf_03}
      boxExtents=<20 20 20>
      constraintSetup {angularLimitsMax=<0 7 20>, angularLimitsMin=<0 -10 0>, angularTarget=<0 0 0>}
      localJointOffset=<5 0 0>
     }]
    sphericalLimits [{limitRadius=7, drivingBone {boneName=spine_02}, sphereLocalOffset=<1.4504 8.8225 11.8554>}]
   }
   n31 = "Transform (Modify) Bone - Bone: dyn_scarf_01"!(ComponentPose=@n30, Alpha:float=Get playing melee(), Translation="0,0.7,0", Rotation="12.337,6.946,-0.044", Scale="1,1,1") -> (Pose) {
    rotationMode=BMM_Additive
    rotationSpace=BCS_BoneSpace
    translationMode=BMM_Additive
    translationSpace=BCS_BoneSpace
    alphaBoolBlend {blendInTime=0.2, blendOption=Linear, blendOutTime=0.2}
    alphaScaleBiasClamp {bInterpResult=true}
    boneToModify {boneName=dyn_scarf_01}
    rotation=<rot 12.337 6.946 -0.0439>
    translation {y=0.7}
   }
   n32 = Get bIsPlayingEmote()
   n33 = "Transform (Modify) Bone - Bone: dyn_scarf_lamahead_head"(ComponentPose=@n31, bAlphaBoolEnabled=@n32, Translation="-2,0,0", Rotation="-9.215,-10.555,11.523", Scale="1,1,1") {
    alphaInputType=Bool
    rotationMode=BMM_Additive
    rotationSpace=BCS_BoneSpace
    translationMode=BMM_Additive
    translationSpace=BCS_BoneSpace
    alphaBoolBlend {blendInTime=0.2, blendOption=Linear, blendOutTime=0.2}
    alphaScaleBiasClamp {bInterpResult=true}
    boneToModify {boneName=dyn_scarf_lamahead_head}
    rotation=<rot -9.215 -10.555 11.5231>
    translation {x=-2}
   }
   n34 = "Transform (Modify) Bone - Bone: dyn_scarf_lama_neck_02"!(ComponentPose=@n33, bAlphaBoolEnabled:bool=@n32, Translation="-1,0,0", Rotation="-1.510,-3.140,1.869", Scale="1,1,1") -> (Pose) {
    alphaInputType=Bool
    rotationMode=BMM_Additive
    rotationSpace=BCS_BoneSpace
    translationMode=BMM_Additive
    translationSpace=BCS_BoneSpace
    alphaBoolBlend {blendInTime=0.2, blendOption=Linear, blendOutTime=0.2}
    alphaScaleBiasClamp {bInterpResult=true}
    boneToModify {boneName=dyn_scarf_lama_neck_02}
    rotation=<rot -1.51 -3.14 1.869>
    translation {x=-1}
   }
   Save cached pose 'scarf'(Pose=Component To Local(ComponentPose=@n34)) {cacheName=scarf}
  ]}
 EventGraph {nodes [
   n1 = Event BlueprintUpdateAnimation()
   > Set Delta Time X(Delta Time X=@n1.DeltaTimeX)
   > n2 = Sequence!(execute) -> (then_0, then_1, then_2:Exec, then_3:Exec)
   > Set r arm to the left(r arm to the left:float=GetAlphaFromReferenceTransformDelta(SkeletalMeshComponent=Get Skel Mesh(), BoneName=upperarm_r, BaseName=spine_04, InputRange=(X=20,Y=90), TwistAxis="1,0,0")) -> (Output_Get:float)
   > n3 = Gravity Override State Function DEPRECATED(CharacterPartAnimInstance=Self-Reference(), GameTimeInSeconds=GetGameTimeInSeconds(), GravityOverrideParams=Get Gravity Override Params())
   > Set Gravity Override(Gravity Override=@n3)
   Set sky diving(execute=@n2.then_1, sky diving:double="To Float (Boolean)"(InBool=Get bIsSkyDiving())) -> (Output_Get:double)
   > Set crouching float(crouching float:double="To Float (Boolean)"(InBool=Get bIsCrouching())) -> (Output_Get:double)
   n4 = AND Boolean(A=Get bIsCrouching(), B=Get bIsMoving2D())
   > Set crouching and moving 2d(crouching and moving 2d:double=FInterpTo(Current=Get crouching and moving 2d(), Target="To Float (Boolean)"(InBool=@n4), DeltaTime=Get Delta Time X(), InterpSpeed=20)) -> (Output_Get:double)
   n5 = Get bIsMoving2D()
   n6 = "To Float (Boolean)"(InBool=AND Boolean(A=Get bIsCrouching(), B=NOT Boolean(A=@n5)))
   > Set crouching not moving(crouching not moving:double=@n6) -> (Output_Get:double)
   n7 = Get bIsKneeSliding()
   > Set is knee sliding(is knee sliding:double=FInterpTo(Current=Get is knee sliding(), Target="To Float (Boolean)"(InBool=@n7), DeltaTime=Get Delta Time X(), InterpSpeed=15)) -> (Output_Get:double)
   n8 = Get bIsBaseballSliding()
   > Set is baseball sliding(is baseball sliding:double=FInterpTo(Current=Get is baseball sliding(), Target="To Float (Boolean)"(InBool=@n8), DeltaTime=Get Delta Time X(), InterpSpeed=15)) -> (Output_Get:double)
   > Set jumping and falling(jumping and falling:bool=OR Boolean(A=Get bIsJumping(), B=Get bIsFalling())) -> (Output_Get:bool)
   > Set turn off gravity override(turn off gravity override:bool=OR Boolean(A=Get bIsSkydiveFloating(), B=Get bIsParachuteOpened())) -> (Output_Get:bool)
   n9 = Get bIsPlayingMeleeAnim()
   > Set playing melee(playing melee:double=FInterpTo(Current=Get playing melee(), Target="To Float (Boolean)"(InBool=@n9), DeltaTime=Get Delta Time X(), InterpSpeed=15)) -> (Output_Get:double)
   > Set is playing emote(is playing emote:double="To Float (Boolean)"(InBool=Get bIsPlayingEmote())) -> (Output_Get:double)
   > Set scarf is on(scarf is on:bool=NOT Boolean(A=Get bIsSkydiveFloating())) -> (Output_Get:bool)
   > Set is in frontend(is in frontend:double="To Float (Boolean)"(InBool=Get bIsInFrontEnd())) -> (Output_Get:double)
   n10 = Montage_IsPlaying(
    self=GetOwnerFortPlayerAnimInstance!(self:Custom Character Part Anim Instance Object Reference) -> (ReturnValue:Fort Player Anim Instance Object Reference)
    Montage=/Game/Animation/Game/MainPlayer/Montages/Melee/PlayerMeleeCombo
   )
   > Set melee fix(melee fix:double=FInterpTo(Current=Get melee fix(), Target="To Float (Boolean)"(InBool=@n10), DeltaTime=Get Delta Time X(), InterpSpeed=15)) -> (Output_Get:double)
   n11 = Event OnCharacterCustomizationLoaded()
   > Set Skel Mesh(Skel Mesh=@n11.CustomBaseMesh)
   FInterpTo(Current=Get crouching and moving 2d(), Target=@n6, DeltaTime=Get Delta Time X(), InterpSpeed=20)
  ]}
]
variables [
 crouching and moving 2d {type=number}
 crouching float {type=number}
 crouching not moving {type=number}
 Delta Time X {type=number}
 Gravity Override {type=Vector}
 Gravity Override Params {
  type=GravityOverrideParamsStruct
  default {
   skydive_DivingDown {gravityOverride=<-3000 -500 -500>}
   skydive_FloatingDown {gravityOverride=<0 -700 -1500>}
   skydive_GoingUp {gravityOverride=<0 300 -1500>}
   skydive_Parachuting {gravityOverride=<-1500 -2000 -980>, windAmplitude=<50 -50 50>, windFrequency=<1.5 3 3>}
  }
 }
 is baseball sliding {type=number}
 is in frontend {type=number}
 is knee sliding {type=number}
 is playing emote {type=number}
 jumping and falling {type=boolean}
 melee fix {type=number}
 playing melee {type=number}
 r arm to the left {type=number}
 scarf is on {type=boolean}
 Skel Mesh {type=SkeletalMeshComponent}
 sky diving {type=number}
 start crouch {type=number}
 turn off gravity override {type=boolean}
]
