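# Control Rig for the F_MED_DollEnthusiast FaceAcc backpack part that applies pose-driven corrective offsets to back hair bones.
# From BeginExecution, a spherical pose reader on pelvis relative to head weights offsets on the left, center and right third hair bones.
# A mirrored shoulder chain, unlinked from BeginExecution, averages upperarm and clavicle readers for side strands and uses clavicle alone for their master bones.
# A final right-clavicle reader is unused.
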ControlRigBlueprint /BRCosmetics/Characters/Player/Female/Medium/Bodies/F_MED_DollEnthusiast/Meshes/Parts/F_MED_DollEnthusiast_FaceAcc_Backpack_CtrlRig
graphs [
 RigVMFunctionLibrary {links [], nodes []}
 RigVMModel {nodes [
   RigUnit_BeginExecution = &RigUnit_BeginExecution()
   n1 = &RigUnit_SphericalPoseReader(
    ActiveRegionScaleFactors=(PositiveWidth=0.4,NegativeWidth=1.2,PositiveHeight=0.4,NegativeHeight=0.4)
    ActiveRegionSize=0.1
    Debug=(bDrawDebug=false,bDraw2D=False,bDrawLocalAxes=False,DebugScale=25,DebugSegments=20,DebugThickness=0.25)
    DriverAxis=(X=1,Y=0,Z=0)
    DriverItem=(Type=Bone,Name="pelvis")
    FalloffRegionScaleFactors=(PositiveWidth=1,NegativeWidth=1,PositiveHeight=1,NegativeHeight=1)
    FalloffSize=0.3
    FlipHeightScaling=False
    FlipWidthScaling=False
    OptionalParentItem=(Type=Bone,Name="head")
    RotationOffset=(X=100,Y=0,Z=-35)
    ExecutePin=@RigUnit_BeginExecution.ExecutePin
   )
   > &RigUnit_ModifyTransforms(
    ItemToModify=((Item=(Type=Bone,Name="dyn_hair_bk_l_3"),Transform=(Rotation=(X=0,Y=0.1736,Z=0,W=0.9848),Translation=(X=6,Y=0,Z=3),Scale3D=(X=1,Y=1,Z=1))))
    WeightMinimum=0
    Weight=@n1.OutputParam
   )
   > &RigUnit_ModifyTransforms(
    ItemToModify=((Item=(Type=Bone,Name="dyn_hair_bk_c_3"),Transform=(Rotation=(X=0,Y=0.1736,Z=0,W=0.9848),Translation=(X=6,Y=0,Z=3),Scale3D=(X=1,Y=1,Z=1))))
    WeightMinimum=0
    Weight=@n1.OutputParam
   )
   > &RigUnit_ModifyTransforms(
    ItemToModify=((Item=(Type=Bone,Name="dyn_hair_bk_r_3"),Transform=(Rotation=(X=0,Y=0.1736,Z=0,W=0.9848),Translation=(X=6,Y=0,Z=3),Scale3D=(X=1,Y=1,Z=1))))
    WeightMinimum=0
    Weight=@n1.OutputParam
   )
   n2 = &RigUnit_SphericalPoseReader(
    ActiveRegionScaleFactors=(PositiveWidth=1,NegativeWidth=1,PositiveHeight=1,NegativeHeight=1.1)
    ActiveRegionSize=0.1
    Debug=(bDrawDebug=false,bDraw2D=False,bDrawLocalAxes=False,DebugScale=25,DebugSegments=20,DebugThickness=0.25)
    DriverAxis=(X=-1,Y=0,Z=0)
    DriverItem=(Type=Bone,Name="upperarm_r")
    ExecutePin=()
    FalloffRegionScaleFactors=(PositiveWidth=0.5,NegativeWidth=0.5,PositiveHeight=1,NegativeHeight=0.6)
    FalloffSize=0.3
    FlipHeightScaling=False
    FlipWidthScaling=False
    OptionalParentItem=(Type=Bone,Name="None")
    RotationOffset=(X=22,Y=65,Z=0)
   )
   > n3 = &RigUnit_SphericalPoseReader(
    ActiveRegionScaleFactors=(PositiveWidth=1,NegativeWidth=1,PositiveHeight=1,NegativeHeight=1)
    ActiveRegionSize=0.1
    Debug=(bDrawDebug=true,bDraw2D=False,bDrawLocalAxes=False,DebugScale=25,DebugSegments=20,DebugThickness=0.25)
    DriverAxis=(X=-1,Y=0,Z=0)
    DriverItem=(Type=Bone,Name="clavicle_r")
    FalloffRegionScaleFactors=(PositiveWidth=0.5,NegativeWidth=0.5,PositiveHeight=0.5,NegativeHeight=0.3)
    FalloffSize=0.3
    FlipHeightScaling=False
    FlipWidthScaling=False
    OptionalParentItem=(Type=Bone,Name="None")
    RotationOffset=(X=30,Y=110,Z=0)
   )
   Add = &RigVMFunction_MathDoubleAdd(A=@n3.OutputParam, B=@n2.OutputParam)
   Divide = &RigVMFunction_MathDoubleDiv(B=2, A=@Add.Result)
   > &RigUnit_ModifyTransforms(
    ItemToModify="((Item=(Type=Bone,Name=\"None\"),Transform=(Rotation=(X=0,Y=0,Z=0,W=1),Translation=(X=0,Y=0,Z=0),Scale3D=(X=1,Y=1,Z=1))),(Item=(Type=Bone,Name=\"dyn_hair_bk_r_3\"),Transform=(Rotation=(X=0,Y=0.0436,Z=0,W=0.999),Translation=(X=0,Y=3,Z=2),Scale3D=(X=1,Y=1,Z=1))) ...(truncated)"
    WeightMinimum=0
    Weight=@Divide.Result
   )
   > &RigUnit_ModifyTransforms(
    ItemToModify=((Item=(Type=Bone,Name="None"),Transform=(Rotation=(X=0,Y=0,Z=0,W=1),Translation=(X=0,Y=0,Z=0),Scale3D=(X=1,Y=1,Z=1))),(Item=(Type=Bone,Name="dyn_hair_bk_r_4"),Transform=(Rotation=(X=0,Y=0,Z=0,W=1),Translation=(X=0,Y=1,Z=5),Scale3D=(X=1,Y=1,Z=1))))
    WeightMinimum=0
    Weight=@Divide.Result
   )
   > n4 = &RigUnit_SphericalPoseReader(
    ActiveRegionScaleFactors=(PositiveWidth=1,NegativeWidth=1,PositiveHeight=1,NegativeHeight=1)
    ActiveRegionSize=0.1
    Debug=(bDrawDebug=true,bDraw2D=False,bDrawLocalAxes=False,DebugScale=25,DebugSegments=20,DebugThickness=0.25)
    DriverAxis=(X=-1,Y=0,Z=0)
    DriverItem=(Type=Bone,Name="clavicle_r")
    FalloffRegionScaleFactors=(PositiveWidth=0.5,NegativeWidth=0.5,PositiveHeight=0.5,NegativeHeight=0.3)
    FalloffSize=0.3
    FlipHeightScaling=False
    FlipWidthScaling=False
    OptionalParentItem=(Type=Bone,Name="None")
    RotationOffset=(X=30,Y=110,Z=0)
   )
   > &RigUnit_ModifyTransforms(
    ItemToModify=((Item=(Type=Bone,Name="None"),Transform=(Rotation=(X=0,Y=0,Z=0,W=1),Translation=(X=0,Y=0,Z=0),Scale3D=(X=1,Y=1,Z=1))),(Item=(Type=Bone,Name="dyn_hair_bk_r_MSTR"),Transform=(Rotation=(X=0,Y=0,Z=0,W=1),Translation=(X=0,Y=1,Z=1),Scale3D=(X=1,Y=1,Z=1))))
    WeightMinimum=0
    Weight=@n4.OutputParam
   )
   > n5 = &RigUnit_SphericalPoseReader(
    ActiveRegionScaleFactors=(PositiveWidth=1,NegativeWidth=1,PositiveHeight=1.1,NegativeHeight=1)
    ActiveRegionSize=0.1
    Debug=(bDrawDebug=false,bDraw2D=False,bDrawLocalAxes=False,DebugScale=25,DebugSegments=20,DebugThickness=0.25)
    DriverAxis=(X=1,Y=0,Z=0)
    DriverItem=(Type=Bone,Name="upperarm_l")
    FalloffRegionScaleFactors=(PositiveWidth=0.5,NegativeWidth=0.5,PositiveHeight=0.6,NegativeHeight=1)
    FalloffSize=0.3
    FlipHeightScaling=False
    FlipWidthScaling=False
    OptionalParentItem=(Type=Bone,Name="None")
    RotationOffset=(X=-22,Y=-115,Z=0)
   )
   > n6 = &RigUnit_SphericalPoseReader(
    ActiveRegionScaleFactors=(PositiveWidth=1,NegativeWidth=1,PositiveHeight=1,NegativeHeight=1)
    ActiveRegionSize=0.1
    Debug=(bDrawDebug=true,bDraw2D=False,bDrawLocalAxes=False,DebugScale=25,DebugSegments=20,DebugThickness=0.25)
    DriverAxis=(X=1,Y=0,Z=0)
    DriverItem=(Type=Bone,Name="clavicle_l")
    FalloffRegionScaleFactors=(PositiveWidth=0.5,NegativeWidth=0.5,PositiveHeight=0.3,NegativeHeight=0.5)
    FalloffSize=0.3
    FlipHeightScaling=False
    FlipWidthScaling=False
    OptionalParentItem=(Type=Bone,Name="None")
    RotationOffset=(X=-30,Y=-70,Z=0)
   )
   n7 = &RigVMFunction_MathDoubleDiv(B=2, A=&RigVMFunction_MathDoubleAdd(A=@n6.OutputParam, B=@n5.OutputParam).Result)
   > &RigUnit_ModifyTransforms(
    ItemToModify="((Item=(Type=Bone,Name=\"None\"),Transform=(Rotation=(X=0,Y=0,Z=0,W=1),Translation=(X=0,Y=0,Z=0),Scale3D=(X=1,Y=1,Z=1))),(Item=(Type=Bone,Name=\"dyn_hair_bk_l_3\"),Transform=(Rotation=(X=0,Y=0.0436,Z=0,W=0.999),Translation=(X=0,Y=-3,Z=2),Scale3D=(X=1,Y=1,Z=1)) ...(truncated)"
    WeightMinimum=0
    Weight=@n7.Result
   )
   > &RigUnit_ModifyTransforms(
    ItemToModify=((Item=(Type=Bone,Name="None"),Transform=(Rotation=(X=0,Y=0,Z=0,W=1),Translation=(X=0,Y=0,Z=0),Scale3D=(X=1,Y=1,Z=1))),(Item=(Type=Bone,Name="dyn_hair_bk_l_4"),Transform=(Rotation=(X=0,Y=0,Z=0,W=1),Translation=(X=0,Y=-1,Z=5),Scale3D=(X=1,Y=1,Z=1))))
    WeightMinimum=0
    Weight=@n7.Result
   )
   > n8 = &RigUnit_SphericalPoseReader(
    ActiveRegionScaleFactors=(PositiveWidth=1,NegativeWidth=1,PositiveHeight=1,NegativeHeight=1)
    ActiveRegionSize=0.1
    Debug=(bDrawDebug=true,bDraw2D=False,bDrawLocalAxes=False,DebugScale=25,DebugSegments=20,DebugThickness=0.25)
    DriverAxis=(X=1,Y=0,Z=0)
    DriverItem=(Type=Bone,Name="clavicle_l")
    FalloffRegionScaleFactors=(PositiveWidth=0.5,NegativeWidth=0.5,PositiveHeight=0.3,NegativeHeight=0.5)
    FalloffSize=0.3
    FlipHeightScaling=False
    FlipWidthScaling=False
    OptionalParentItem=(Type=Bone,Name="None")
    RotationOffset=(X=-30,Y=-70,Z=0)
   )
   > &RigUnit_ModifyTransforms(
    ItemToModify="((Item=(Type=Bone,Name=\"None\"),Transform=(Rotation=(X=0,Y=0,Z=0,W=1),Translation=(X=0,Y=0,Z=0),Scale3D=(X=1,Y=1,Z=1))),(Item=(Type=Bone,Name=\"dyn_hair_bk_l_MSTR\"),Transform=(Rotation=(X=0,Y=0,Z=0,W=1),Translation=(X=0,Y=-1,Z=4.3175),Scale3D=(X=1,Y=1,Z=1))) ...(truncated)"
    WeightMinimum=0
    Weight=@n8.OutputParam
   )
   &RigUnit_SphericalPoseReader(
    ActiveRegionScaleFactors=(PositiveWidth=1,NegativeWidth=1,PositiveHeight=1,NegativeHeight=1)
    ActiveRegionSize=0.02
    Debug=(bDrawDebug=false,bDraw2D=False,bDrawLocalAxes=False,DebugScale=25,DebugSegments=20,DebugThickness=0.25)
    DriverAxis=(X=-1,Y=0,Z=0)
    DriverItem=(Type=Bone,Name="clavicle_r")
    ExecutePin=()
    FalloffRegionScaleFactors=(PositiveWidth=1,NegativeWidth=1,PositiveHeight=1,NegativeHeight=1)
    FalloffSize=0.14
    FlipHeightScaling=False
    FlipWidthScaling=False
    OptionalParentItem=(Type=Bone,Name="head")
    RotationOffset=(X=180,Y=-23,Z=0)
   )
  ]}
]
hierarchy [
 root {loc=<0 0 0>, rot=<0 0 0 1>}
 pelvis {parent=root, loc=<0 -0.5301 98.0571>, rot=<0.0127 0.707 -0.0127 -0.707>}
 spine_01 {parent=pelvis, loc=<6.1208 -0.9328 0>, rot=<0 0 0.034 -0.9994>}
 spine_02 {parent=spine_01, loc=<6.5996 -0.2095 0>, rot=<0 0 0.0238 -0.9997>}
 spine_03 {parent=spine_02, loc=<6.4914 -0.1237 0>, rot=<0 0 0.0149 -0.9999>}
 spine_04 {parent=spine_03, loc=<6.9598 0.0794 0>, rot=<0 0 -0.0132 -0.9999>}
 spine_05 {parent=spine_04, loc=<6.9987 0 0>, rot=<0 0 -0.0015 -1>}
 clavicle_l {parent=spine_05, loc=<5.4831 -0.1053 -3.2946>, rot=<0.0465 -0.7289 -0.009 -0.6829>}
 upperarm_l {parent=clavicle_l, loc=<9.8658 -1.2136 -0.4105>, rot=<0.0164 -0.3245 -0.0269 -0.9454>}
 neck_01 {parent=spine_05, loc=<11.3138 0.1529 0>, rot=<0 0 -0.2003 -0.9797>}
 neck_02 {parent=neck_01, loc=<4.5627 0 0>, rot=<0 0 0.0504 -0.9987>}
 head {parent=neck_02, loc=<5.3914 0 0>, rot=<0 0 0.115 -0.9934>}
 dyn_hair_side_bang_r_1 {parent=head, loc=<4.1266 2.3782 6.9508>, rot=<0 0 1 0>}
 dyn_hair_side_bang_r_2 {parent=dyn_hair_side_bang_r_1, loc=<5.1022 -0.0844 -0.2788>, rot=<0 0 0 1>}
 dyn_hair_side_bang_r_fr_1 {parent=head, loc=<8.7513 5.8959 6.2631>, rot=<0 0.2654 0.9641 0>}
 dyn_hair_fr_bang_r_1 {parent=head, loc=<8.9635 9.5516 1.3768>, rot=<0.1277 0.6205 0.7619 0.135>}
 dyn_hair_fr_bang_r_2_r {parent=dyn_hair_fr_bang_r_1, loc=<2.6592 1.3856 -0.1572>, rot=<0 -0.1737 0 -0.9848>}
 dyn_hair_fr_bang_r_2_l {parent=dyn_hair_fr_bang_r_1, loc=<2.9077 -1.8997 -0.3989>, rot=<-0.1017 -0.1621 0.0025 -0.9815>}
 dyn_hair_fr_bang_l_1 {parent=head, loc=<9.2091 7.2152 -5.0378>, rot=<-0.0414 -0.9226 -0.3559 -0.1429>}
 dyn_hair_front_bang_l_2 {parent=dyn_hair_fr_bang_l_1, loc=<2.8384 0.1051 -0.0987>, rot=<0 0 0 1>}
 dyn_hair_side_bang_l_1 {parent=head, loc=<4.0581 2.4789 -6.9997>, rot=<0 -1 0 0>}
 dyn_hair_side_bang_l_2 {parent=dyn_hair_side_bang_l_1, loc=<4.9833 0.0632 -0.3341>, rot=<0 0 0 1>}
 dyn_hair_bk_c_MSTR {parent=head, loc=<9.6909 -6.2732 0>, rot=<0.0624 -0.7043 0.7043 -0.0624>}
 dyn_hair_bk_c_1 {parent=dyn_hair_bk_c_MSTR, loc=<0 0 0>, rot=<0 0 0 1>}
 dyn_hair_bk_c_2 {parent=dyn_hair_bk_c_1, loc=<9.5622 0 -0.0073>, rot=<0 0.0203 0 -0.9998>}
 dyn_hair_bk_c_3 {parent=dyn_hair_bk_c_2, loc=<5.0221 0.0523 0.1175>, rot=<0 0.06 0 -0.9982>}
 dyn_hair_bk_c_4 {parent=dyn_hair_bk_c_3, loc=<6.6814 -0.0637 -0.3613>, rot=<0 0 0 1>}
 dyn_hair_bk_c_5 {parent=dyn_hair_bk_c_4, loc=<6.7816 -0.0637 -0.5306>, rot=<0 0 0 1>}
 dyn_hair_bk_c_6 {parent=dyn_hair_bk_c_5, loc=<7.3448 0.0752 -0.3733>, rot=<0 0 0 1>}
 dyn_hair_bk_r_MSTR {parent=head, loc=<9.6732 -4.1317 5.0873>, rot=<0.0301 -0.7065 0.7065 -0.0301>}
 dyn_hair_bk_r_1 {parent=dyn_hair_bk_r_MSTR, loc=<0 0 0>, rot=<0 0 0 1>}
 dyn_hair_bk_r_2 {parent=dyn_hair_bk_r_1, loc=<9.619 -0.3081 1.7165>, rot=<0.0065 0.131 0.0009 -0.9914>}
 dyn_hair_bk_r_3 {parent=dyn_hair_bk_r_2, loc=<5.1299 -0.4129 -0.1726>, rot=<-0.0671 0.0312 0.0267 -0.9969>}
 dyn_hair_bk_r_4 {parent=dyn_hair_bk_r_3, loc=<6.6942 -0.6381 -0.7621>, rot=<0 0 0.0266 -0.9996>}
 dyn_hair_bk_r_5 {parent=dyn_hair_bk_r_4, loc=<6.8712 -0.2346 -0.7301>, rot=<-0.0361 0.001 -0.0224 -0.9991>}
 dyn_hair_bk_r_6 {parent=dyn_hair_bk_r_5, loc=<7.6856 -0.9896 -0.4086>, rot=<-0.0233 0.0018 0.0777 -0.9967>}
 dyn_hair_bk_l_MSTR {parent=head, loc=<9.6733 -4.1317 -5.0873>, rot=<0.0574 -0.7048 0.7013 -0.0903>}
 dyn_hair_bk_l_1 {parent=dyn_hair_bk_l_MSTR, loc=<0 0 0>, rot=<0 0 0 1>}
 dyn_hair_bk_l_2 {parent=dyn_hair_bk_l_1, loc=<9.6996 0 0>, rot=<0.0014 0.0613 0.0233 -0.9978>}
 dyn_hair_bk_l_3 {parent=dyn_hair_bk_l_2, loc=<5.1167 0.1888 -0.0538>, rot=<0.1064 0.0262 -0.0548 -0.9925>}
 dyn_hair_bk_l_4 {parent=dyn_hair_bk_l_3, loc=<6.9313 -0.2793 0.0311>, rot=<0 0 -0.0302 -0.9995>}
 dyn_hair_bk_l_5 {parent=dyn_hair_bk_l_4, loc=<6.892 -0.011 -0.4767>, rot=<0.0081 0.0081 0.0177 -0.9998>}
 dyn_hair_bk_l_6 {parent=dyn_hair_bk_l_5, loc=<7.7474 0.2985 -0.0533>, rot=<0 0 0 1>}
 dyn_earring_r_1 {parent=head, loc=<-0.3653 0.1171 6.436>, rot=<0.014 0.4115 0.9113 0.0007>}
 dyn_earring_l_1 {parent=head, loc=<-0.3652 0.1171 -6.436>, rot=<0.0035 -0.9114 -0.4115 -0.0047>}
 dyn_earring_l_2 {parent=head, loc=<-0.6051 -0.5985 -7.2448>, rot=<0 -1 0 0>}
 clavicle_r {parent=spine_05, loc=<5.4831 -0.1053 3.2946>, rot=<0.7289 0.0465 -0.6829 0.009>}
 upperarm_r {parent=clavicle_r, loc=<-9.8658 1.2136 0.4105>, rot=<0.0164 -0.3245 -0.0269 -0.9454>}
]
