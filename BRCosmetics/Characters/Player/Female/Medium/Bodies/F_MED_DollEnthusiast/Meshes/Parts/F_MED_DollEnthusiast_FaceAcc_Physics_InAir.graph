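# Physics asset for the DollEnthusiast female character's FaceAcc mesh, simulating back hair and named as the in-air variant.
# Three hair chains (center, left, right) each run six capsule segments from a MSTR root, with tight swing and twist limits.
# Capsules on head, neck, spine, clavicles and upper arms provide body collision shapes.
# Angularly free cross-constraints tie the left and right fifth segments to the center one, linking the strands.
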PhysicsAsset /BRCosmetics/Characters/Player/Female/Medium/Bodies/F_MED_DollEnthusiast/Meshes/Parts/F_MED_DollEnthusiast_FaceAcc_Physics_InAir
bodies {
 clavicle_l [{shapeType=Capsule, length=15, radius=7.5, shapeName=clavicle_l_capsule, center=<5.5365 0 -2.6289>, rotation=<rot 74.5828 -0.0019 -0.002>}]
 clavicle_r [{shapeType=Capsule, length=7, radius=7.5, shapeName=clavicle_r_capsule, center=<-4.5802 0 2.4478>, rotation=<rot 80.8939 -0.0044 -9.027>}]
 dyn_hair_bk_c_1 [{shapeType=Capsule, length=6, radius=4.5, shapeName=dyn_hair_bk_c_1_capsule, center=<3.6825 0 0>, rotation=<rot 0 0 90>}]
 dyn_hair_bk_c_2 [{shapeType=Capsule, length=5, radius=4, shapeName=dyn_hair_bk_c_2_capsule, center=<2.3751 0 1.0107>, rotation=<rot 0 90 101.1949>}]
 dyn_hair_bk_c_3 [{shapeType=Capsule, length=4, radius=4, shapeName=dyn_hair_bk_c_3_capsule, center=<2.3751 0 1.0107>, rotation=<rot 0 0 90>}]
 dyn_hair_bk_c_4 [{shapeType=Capsule, length=1, radius=7, shapeName=dyn_hair_bk_c_4_capsule, center=<2.3751 0 4.5735>, rotation=<rot 0 0 90>}]
 dyn_hair_bk_c_5 [{shapeType=Capsule, length=2, radius=6, shapeName=dyn_hair_bk_c_5_capsule, center=<3.8137 0 3.7481>, rotation=<rot 0 0 90>}]
 dyn_hair_bk_c_6 [{shapeType=Capsule, length=7, radius=5.5, shapeName=dyn_hair_bk_c_6_capsule, center=<2.3751 0 2.648>, rotation=<rot 0 0 90>}]
 dyn_hair_bk_c_MSTR [{shapeType=Capsule, length=1, radius=0.5, shapeName=dyn_hair_bk_c_MSTR_capsule, rotation=<rot 0 0 90>}]
 dyn_hair_bk_l_1 [{shapeType=Capsule, length=6, radius=4.5, shapeName=dyn_hair_bk_l_1_capsule, center=<3.6825 0 0>, rotation=<rot 0 0 90>}]
 dyn_hair_bk_l_2 [{
   shapeType=Capsule
   length=5
   radius=5
   shapeName=dyn_hair_bk_l_2_capsule
   center=<2.3272 0.001 1.4565>
   rotation=<rot -1.1281 99.4098 96.0288>
  }]
 dyn_hair_bk_l_3 [{shapeType=Capsule, length=4, radius=4, shapeName=dyn_hair_bk_l_3_capsule, center=<2.3751 0 1.7993>, rotation=<rot 0 90 90>}]
 dyn_hair_bk_l_4 [{shapeType=Capsule, length=1, radius=7, shapeName=dyn_hair_bk_l_4_capsule, center=<2.3751 0 4.5735>, rotation=<rot 0 0 90>}]
 dyn_hair_bk_l_5 [{shapeType=Capsule, length=2, radius=6, shapeName=dyn_hair_bk_l_5_capsule, center=<3.8137 0 3.7481>, rotation=<rot 0 0 90>}]
 dyn_hair_bk_l_6 [{shapeType=Capsule, length=7, radius=5.5, shapeName=dyn_hair_bk_l_6_capsule, center=<2.3751 0 2.648>, rotation=<rot 0 0 90>}]
 dyn_hair_bk_l_MSTR [{shapeType=Capsule, length=1, radius=0.5, shapeName=dyn_hair_bk_l_MSTR_capsule, rotation=<rot 0 0 90>}]
 dyn_hair_bk_r_1 [{shapeType=Capsule, length=6, radius=4.5, shapeName=dyn_hair_bk_r_1_capsule, center=<3.6825 0 0>, rotation=<rot 0 0 90>}]
 dyn_hair_bk_r_2 [{shapeType=Capsule, length=5, radius=5, shapeName=dyn_hair_bk_r_2_capsule, center=<2.3751 0 1.0107>, rotation=<rot 1.0603 80.7676 96.4946>}]
 dyn_hair_bk_r_3 [{shapeType=Capsule, length=4, radius=4, shapeName=dyn_hair_bk_r_3_capsule, center=<2.3751 0 1.3312>, rotation=<rot 0 90 90>}]
 dyn_hair_bk_r_4 [{shapeType=Capsule, length=1, radius=7, shapeName=dyn_hair_bk_r_4_capsule, center=<2.3751 0 4.5735>, rotation=<rot 0 0 90>}]
 dyn_hair_bk_r_5 [{shapeType=Capsule, length=2, radius=6, shapeName=dyn_hair_bk_r_5_capsule, center=<3.8137 0 3.7481>, rotation=<rot 0 0 90>}]
 dyn_hair_bk_r_6 [{shapeType=Capsule, length=7, radius=5.5, shapeName=dyn_hair_bk_r_6_capsule, center=<2.3751 0 2.648>, rotation=<rot 0 0 90>}]
 dyn_hair_bk_r_MSTR [{shapeType=Capsule, length=1, radius=0.5, shapeName=dyn_hair_bk_r_MSTR_capsule, rotation=<rot 0 0 90>}]
 head [{shapeType=Capsule, length=3, radius=2, shapeName=head_capsule, center=<3.152 0 0>, rotation=<rot 90 0 0>}]
 neck_01 [{shapeType=Capsule, length=4.5, radius=4, shapeName=neck_01_capsule, center=<5.9102 0 0>, rotation=<rot 90 0 0>}]
 spine_02 [{shapeType=Capsule, length=6, radius=13, shapeName=spine_02_capsule, center=<3.0763 4.544 0>}]
 spine_03 [{shapeType=Capsule, length=10, radius=12.5, shapeName=spine_03_capsule, center=<3.5235 4.1187 0>}]
 spine_04 [{shapeType=Capsule, length=12, radius=15, shapeName=spine_04_capsule, center=<3.5208 5.1586 0>}]
 upperarm_l [{shapeType=Capsule, length=21, radius=7, shapeName=upperarm_l_capsule, center=<18.5292 0 -1.0767>, rotation=<rot -90 0 0>}]
 upperarm_r [{shapeType=Capsule, length=25.5, radius=7, shapeName=upperarm_r_capsule, center=<-16.5886 0 1.0767>, rotation=<rot 90 0 0>}]
}
constraints [
 {bone1Name=dyn_hair_bk_c_2, bone2Name=dyn_hair_bk_c_1, swing1LimitDegrees=2, swing2LimitDegrees=20, twistLimitDegrees=1}
 {bone1Name=dyn_hair_bk_c_4, bone2Name=dyn_hair_bk_c_3, swing1LimitDegrees=2, swing2LimitDegrees=10, twistLimitDegrees=1}
 {bone1Name=dyn_hair_bk_c_5, bone2Name=dyn_hair_bk_c_4, swing1LimitDegrees=2, swing2LimitDegrees=20, twistLimitDegrees=1}
 {bone1Name=dyn_hair_bk_c_6, bone2Name=dyn_hair_bk_c_5, swing1LimitDegrees=2, swing2LimitDegrees=20, twistLimitDegrees=1}
 {bone1Name=dyn_hair_bk_l_2, bone2Name=dyn_hair_bk_l_1, swing1LimitDegrees=2, swing2LimitDegrees=20, twistLimitDegrees=1}
 {bone1Name=dyn_hair_bk_l_4, bone2Name=dyn_hair_bk_l_3, swing1LimitDegrees=2, swing2LimitDegrees=10, twistLimitDegrees=1}
 {bone1Name=dyn_hair_bk_l_5, bone2Name=dyn_hair_bk_l_4, swing1LimitDegrees=2, swing2LimitDegrees=20, twistLimitDegrees=1}
 {bone1Name=dyn_hair_bk_l_6, bone2Name=dyn_hair_bk_l_5, swing1LimitDegrees=2, swing2LimitDegrees=20, twistLimitDegrees=1}
 {bone1Name=dyn_hair_bk_r_6, bone2Name=dyn_hair_bk_r_5, swing1LimitDegrees=2, swing2LimitDegrees=20, twistLimitDegrees=1}
 {bone1Name=dyn_hair_bk_r_5, bone2Name=dyn_hair_bk_r_4, swing1LimitDegrees=2, swing2LimitDegrees=20, twistLimitDegrees=1}
 {bone1Name=dyn_hair_bk_r_4, bone2Name=dyn_hair_bk_r_3, swing1LimitDegrees=2, swing2LimitDegrees=10, twistLimitDegrees=1}
 {bone1Name=dyn_hair_bk_r_2, bone2Name=dyn_hair_bk_r_1, swing1LimitDegrees=2, swing2LimitDegrees=20, twistLimitDegrees=1}
 {bone1Name=dyn_hair_bk_l_3, bone2Name=dyn_hair_bk_l_2, swing1LimitDegrees=2, swing2LimitDegrees=20, twistLimitDegrees=2}
 {bone1Name=dyn_hair_bk_c_3, bone2Name=dyn_hair_bk_c_2, swing1LimitDegrees=2, swing2LimitDegrees=20, twistLimitDegrees=2}
 {bone1Name=dyn_hair_bk_r_3, bone2Name=dyn_hair_bk_r_2, swing1LimitDegrees=2, swing2LimitDegrees=20, twistLimitDegrees=2}
 {bone1Name=dyn_hair_bk_c_1, bone2Name=dyn_hair_bk_c_MSTR, swing1LimitDegrees=1, swing2LimitDegrees=10, twistLimitDegrees=1}
 {bone1Name=dyn_hair_bk_l_1, bone2Name=dyn_hair_bk_l_MSTR, swing1LimitDegrees=1, swing2LimitDegrees=10, twistLimitDegrees=1}
 {bone1Name=dyn_hair_bk_r_1, bone2Name=dyn_hair_bk_r_MSTR, swing1LimitDegrees=1, swing2LimitDegrees=10, twistLimitDegrees=1}
 {
  bone1Name=dyn_hair_bk_l_5
  bone2Name=dyn_hair_bk_c_5
  swing1LimitDegrees=45
  swing1Motion=Free
  swing2LimitDegrees=45
  swing2Motion=Free
  twistLimitDegrees=45
  twistMotion=Free
 }
 {
  bone1Name=dyn_hair_bk_r_5
  bone2Name=dyn_hair_bk_c_5
  swing1LimitDegrees=45
  swing1Motion=Free
  swing2LimitDegrees=45
  swing2Motion=Free
  twistLimitDegrees=45
  twistMotion=Free
 }
]
