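# Animation Blueprint for the DollEnthusiast female face-accessory part, adding secondary motion to hair and earrings.
# The parent mesh pose runs through a crouch-aware Control Rig, then AnimDynamics bangs and earrings into a cached pose.
# That pose feeds RigidBody simulation, with a separate physics asset for skydiving.
# When rigid bodies are disabled or non-cape backbling is worn, back-hair bone constraints and a backpack Control Rig replace RigidBody.
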AnimBlueprint /BRCosmetics/Characters/Player/Female/Medium/Bodies/F_MED_DollEnthusiast/Meshes/Parts/F_MED_DollEnthusiast_FaceAcc_AnimBP
target_skeleton=&/F_MED_DollEnthusiast_FaceAcc_Skeleton
graphs [
 AnimGraph {nodes [
   n1 = Use cached pose 'AnimBPPose'()
   Blend Poses by bool(BlendPose_0=Component To Local(ComponentPose=Local To Component(LocalPose=@n1)), bActiveValue=Get bIsSkydiveDiveMode()) {blendType=Linear}
   n2 = "Transform (Modify) Bone - Bone: dyn_hair_bk_l_MSTR"!(ComponentPose, Alpha=1, Rotation="0,0,-20") -> (Pose) {rotationMode=BMM_Additive, alphaBoolBlend {blendOption=Linear}, boneToModify {boneName=dyn_hair_bk_l_MSTR}, rotation {roll=-20}}
   n3 = "Transform (Modify) Bone - Bone: dyn_hair_bk_c_MSTR"!(ComponentPose=@n2, Alpha=1, Rotation="0,0,-20") -> (Pose) {rotationMode=BMM_Additive, alphaBoolBlend {blendOption=Linear}, boneToModify {boneName=dyn_hair_bk_c_MSTR}, rotation {roll=-20}}
   "Transform (Modify) Bone - Bone: dyn_hair_bk_r_MSTR"!(ComponentPose=@n3, Alpha=1, Rotation="0,0,-20") -> (Pose) {rotationMode=BMM_Additive, alphaBoolBlend {blendOption=Linear}, boneToModify {boneName=dyn_hair_bk_r_MSTR}, rotation {roll=-20}}
   Output Pose(Result=Blend Poses by bool(BlendPose_0=Use cached pose 'WithoutRBPose'(), BlendPose_1=Use cached pose 'RBPose'(), bActiveValue=Get Without_RB() -> (Without_RB:bool)) {blendType=Linear}) {layerGroup=None}
   n4 = Anim Dynamics - Bone: dyn_hair_side_bang_l_1!(ComponentPose=Local To Component(LocalPose=Use cached pose 'InputPose'()), Alpha=1, GravityOverride:Vector="0,0,-980", ExternalForce="0,0,0") -> (Pose) {
    angularDampingOverride=0.999
    angularSpringConstant=1000
    bAngularSpring=true
    bOverrideAngularDamping=true
    bUseGravityOverride=true
    gravityScale=0.35
    simulationSpace=RootRelative
    alphaBoolBlend {blendOption=Linear}
    boundBone {boneName=dyn_hair_side_bang_l_1}
    componentAppliedLinearAccClamp=<1000 1000 1000>
    componentLinearAccScale=<0.6 0.6 0.6>
    componentLinearVelScale=<0.5 0.5 0.5>
    gravityOverride {z=-980}
    physicsBodyDefinitions [{
      boundBone {boneName=dyn_hair_side_bang_l_1}
      boxExtents=<15 15 15>
      constraintSetup {angularLimitsMax=<5 0 30>, angularLimitsMin=<-5 -20 -5>, angularTarget=<1 0 0>}
      localJointOffset=<1.5 0 0>
     }]
    relativeSpaceBone {boneName=head}
   }
   n5 = Anim Dynamics - Bone: dyn_hair_side_bang_l_2!(ComponentPose=@n4, Alpha=1, GravityOverride:Vector="0,0,-980", ExternalForce="0,0,0") -> (Pose) {
    angularDampingOverride=0.999
    angularSpringConstant=1000
    bAngularSpring=true
    bOverrideAngularDamping=true
    bUseGravityOverride=true
    gravityScale=0.35
    simulationSpace=RootRelative
    alphaBoolBlend {blendOption=Linear}
    boundBone {boneName=dyn_hair_side_bang_l_2}
    componentAppliedLinearAccClamp=<1000 1000 1000>
    componentLinearAccScale=<0.6 0.6 0.6>
    componentLinearVelScale=<0.5 0.5 0.5>
    gravityOverride {z=-980}
    physicsBodyDefinitions [{
      boundBone {boneName=dyn_hair_side_bang_l_2}
      boxExtents=<10 10 10>
      constraintSetup {angularLimitsMax=<5 5 10>, angularLimitsMin=<-5 -10 -5>, angularTarget=<1 0 0>}
      localJointOffset=<1 0 0>
     }]
    relativeSpaceBone {boneName=head}
   }
   n6 = Anim Dynamics - Bone: dyn_hair_side_bang_r_1!(ComponentPose=@n5, Alpha=1, GravityOverride:Vector="0,0,-980", ExternalForce="0,0,0") -> (Pose) {
    angularDampingOverride=0.999
    angularSpringConstant=1000
    bAngularSpring=true
    bOverrideAngularDamping=true
    bUseGravityOverride=true
    gravityScale=0.35
    simulationSpace=RootRelative
    alphaBoolBlend {blendOption=Linear}
    boundBone {boneName=dyn_hair_side_bang_r_1}
    componentAppliedLinearAccClamp=<1000 1000 1000>
    componentLinearAccScale=<0.6 0.6 0.6>
    componentLinearVelScale=<0.5 0.5 0.5>
    gravityOverride {z=-980}
    physicsBodyDefinitions [{
      boundBone {boneName=dyn_hair_side_bang_r_1}
      boxExtents=<15 15 15>
      constraintSetup {angularLimitsMax=<5 0 5>, angularLimitsMin=<-5 -20 -30>, angularTarget=<1 0 0>}
      localJointOffset=<1.5 0 0>
     }]
    relativeSpaceBone {boneName=head}
   }
   n7 = Anim Dynamics - Bone: dyn_hair_side_bang_r_2!(ComponentPose=@n6, Alpha=1, GravityOverride:Vector="0,0,-980", ExternalForce="0,0,0") -> (Pose) {
    angularDampingOverride=0.999
    angularSpringConstant=1000
    bAngularSpring=true
    bOverrideAngularDamping=true
    bUseGravityOverride=true
    gravityScale=0.35
    simulationSpace=RootRelative
    alphaBoolBlend {blendOption=Linear}
    boundBone {boneName=dyn_hair_side_bang_r_2}
    componentAppliedLinearAccClamp=<1000 1000 1000>
    componentLinearAccScale=<0.6 0.6 0.6>
    componentLinearVelScale=<0.5 0.5 0.5>
    gravityOverride {z=-980}
    physicsBodyDefinitions [{
      boundBone {boneName=dyn_hair_side_bang_r_2}
      boxExtents=<10 10 10>
      constraintSetup {angularLimitsMax=<5 5 5>, angularLimitsMin=<-5 -10 -10>, angularTarget=<1 0 0>}
      localJointOffset=<1 0 0>
     }]
    relativeSpaceBone {boneName=head}
   }
   n8 = Anim Dynamics - Bone: dyn_hair_fr_bang_l_1!(ComponentPose=@n7, Alpha=1, GravityOverride:Vector="0,0,-980", ExternalForce="0,0,0") -> (Pose) {
    angularDampingOverride=0.999
    angularSpringConstant=1000
    bAngularSpring=true
    bOverrideAngularDamping=true
    bUseGravityOverride=true
    gravityScale=0.35
    simulationSpace=RootRelative
    alphaBoolBlend {blendOption=Linear}
    boundBone {boneName=dyn_hair_fr_bang_l_1}
    componentAppliedLinearAccClamp=<1000 1000 1000>
    componentLinearAccScale=<0.6 0.6 0.6>
    componentLinearVelScale=<0.5 0.5 0.5>
    gravityOverride {z=-980}
    physicsBodyDefinitions [{
      boundBone {boneName=dyn_hair_fr_bang_l_1}
      boxExtents=<20 20 15>
      constraintSetup {angularLimitsMax=<5 0 20>, angularLimitsMin=<-5 -20 -10>, angularTarget=<1 0 0>}
      localJointOffset=<1.5 0 0>
     }]
    relativeSpaceBone {boneName=head}
   }
   n9 = Anim Dynamics - Bone: dyn_hair_front_bang_l_2!(ComponentPose=@n8, Alpha=1, GravityOverride:Vector="0,0,-980", ExternalForce="0,0,0") -> (Pose) {
    angularDampingOverride=0.999
    angularSpringConstant=1000
    bAngularSpring=true
    bOverrideAngularDamping=true
    bUseGravityOverride=true
    gravityScale=0.35
    simulationSpace=RootRelative
    alphaBoolBlend {blendOption=Linear}
    boundBone {boneName=dyn_hair_front_bang_l_2}
    componentAppliedLinearAccClamp=<1000 1000 1000>
    componentLinearAccScale=<0.6 0.6 0.6>
    componentLinearVelScale=<0.5 0.5 0.5>
    gravityOverride {z=-980}
    physicsBodyDefinitions [{
      boundBone {boneName=dyn_hair_front_bang_l_2}
      boxExtents=<15 15 10>
      constraintSetup {angularLimitsMax=<5 5 5>, angularLimitsMin=<-5 -10 -10>, angularTarget=<1 0 0>}
      localJointOffset=<1 0 0>
     }]
    relativeSpaceBone {boneName=head}
   }
   n10 = Anim Dynamics - Bone: dyn_hair_side_bang_r_fr_1!(ComponentPose=@n9, Alpha=1, GravityOverride:Vector="0,0,-980", ExternalForce="0,0,0") -> (Pose) {
    angularDampingOverride=0.999
    angularSpringConstant=1000
    bAngularSpring=true
    bOverrideAngularDamping=true
    bUseGravityOverride=true
    gravityScale=0.35
    simulationSpace=RootRelative
    alphaBoolBlend {blendOption=Linear}
    boundBone {boneName=dyn_hair_side_bang_r_fr_1}
    componentAppliedLinearAccClamp=<1000 1000 1000>
    componentLinearAccScale=<0.6 0.6 0.6>
    componentLinearVelScale=<0.5 0.5 0.5>
    gravityOverride {z=-980}
    physicsBodyDefinitions [{
      boundBone {boneName=dyn_hair_side_bang_r_fr_1}
      boxExtents=<15 15 15>
      constraintSetup {angularLimitsMax=<5 0 10>, angularLimitsMin=<-5 -20 -30>, angularTarget=<1 0 0>}
      localJointOffset=<1.2 0 0>
     }]
    relativeSpaceBone {boneName=head}
   }
   n11 = Anim Dynamics - Bone: dyn_hair_fr_bang_r_1!(ComponentPose=@n10, Alpha=1, GravityOverride:Vector="0,0,-980", ExternalForce="0,0,0") -> (Pose) {
    angularDampingOverride=0.999
    angularSpringConstant=1000
    bAngularSpring=true
    bOverrideAngularDamping=true
    bUseGravityOverride=true
    gravityScale=0.35
    simulationSpace=RootRelative
    alphaBoolBlend {blendOption=Linear}
    boundBone {boneName=dyn_hair_fr_bang_r_1}
    componentAppliedLinearAccClamp=<1000 1000 1000>
    componentLinearAccScale=<0.6 0.6 0.6>
    componentLinearVelScale=<0.5 0.5 0.5>
    gravityOverride {z=-980}
    physicsBodyDefinitions [{
      boundBone {boneName=dyn_hair_fr_bang_r_1}
      boxExtents=<15 15 15>
      constraintSetup {angularLimitsMax=<5 0 10>, angularLimitsMin=<-5 -30 -10>, angularTarget=<1 0 0>}
      localJointOffset=<2 0 0>
     }]
    relativeSpaceBone {boneName=head}
   }
   n12 = Anim Dynamics - Bone: dyn_hair_fr_bang_r_2_r!(ComponentPose=@n11, Alpha=1, GravityOverride:Vector="0,0,-980", ExternalForce="0,0,0") -> (Pose) {
    angularDampingOverride=0.999
    angularSpringConstant=1000
    bAngularSpring=true
    bOverrideAngularDamping=true
    bUseGravityOverride=true
    gravityScale=0.35
    simulationSpace=RootRelative
    alphaBoolBlend {blendOption=Linear}
    boundBone {boneName=dyn_hair_fr_bang_r_2_r}
    componentAppliedLinearAccClamp=<1000 1000 1000>
    componentLinearAccScale=<0.6 0.6 0.6>
    componentLinearVelScale=<0.5 0.5 0.5>
    gravityOverride {z=-980}
    physicsBodyDefinitions [{
      boundBone {boneName=dyn_hair_fr_bang_r_2_r}
      boxExtents=<15 10 10>
      constraintSetup {angularLimitsMax=<5 5 20>, angularLimitsMin=<-5 -20 -10>, angularTarget=<1 0 0>}
      localJointOffset=<1.5 0 0>
     }]
    relativeSpaceBone {boneName=head}
   }
   n13 = Anim Dynamics - Bone: dyn_hair_fr_bang_r_2_l!(ComponentPose=@n12, Alpha=1, GravityOverride:Vector="0,0,-980", ExternalForce="0,0,0") -> (Pose) {
    angularDampingOverride=0.999
    angularSpringConstant=1000
    bAngularSpring=true
    bOverrideAngularDamping=true
    bUseGravityOverride=true
    gravityScale=0.35
    simulationSpace=RootRelative
    alphaBoolBlend {blendOption=Linear}
    boundBone {boneName=dyn_hair_fr_bang_r_2_l}
    componentAppliedLinearAccClamp=<1000 1000 1000>
    componentLinearAccScale=<0.6 0.6 0.6>
    componentLinearVelScale=<0.5 0.5 0.5>
    gravityOverride {z=-980}
    physicsBodyDefinitions [{
      boundBone {boneName=dyn_hair_fr_bang_r_2_l}
      boxExtents=<20 20 10>
      constraintSetup {angularLimitsMax=<5 5 10>, angularLimitsMin=<-5 -20 -10>, angularTarget=<1 0 0>}
      localJointOffset=<1 0 0>
     }]
    relativeSpaceBone {boneName=head}
   }
   n14 = Anim Dynamics - Bone: dyn_earring_r_1!(ComponentPose=@n13, Alpha=1, GravityOverride:Vector="0,0,-980", ExternalForce="0,0,0") -> (Pose) {
    angularDampingOverride=0.999
    angularSpringConstant=500
    bAngularSpring=true
    bOverrideAngularDamping=true
    bUseGravityOverride=true
    gravityScale=0.35
    simulationSpace=RootRelative
    alphaBoolBlend {blendOption=Linear}
    boundBone {boneName=dyn_earring_r_1}
    componentAppliedLinearAccClamp=<1000 1000 1000>
    componentLinearAccScale=<0.6 0.6 0.6>
    componentLinearVelScale=<0.5 0.5 0.5>
    gravityOverride {z=-980}
    physicsBodyDefinitions [{
      boundBone {boneName=dyn_earring_r_1}
      boxExtents=<15 15 15>
      constraintSetup {angularLimitsMax=<0 30 30>, angularLimitsMin=<0 -30 -30>, angularTarget=<1 0 0>}
      localJointOffset=<2 0 0>
     }]
    relativeSpaceBone {boneName=head}
   }
   n15 = Anim Dynamics - Bone: dyn_earring_l_1!(ComponentPose=@n14, Alpha=1, GravityOverride:Vector="0,0,-980", ExternalForce="0,0,0") -> (Pose) {
    angularDampingOverride=0.999
    angularSpringConstant=500
    bAngularSpring=true
    bOverrideAngularDamping=true
    bUseGravityOverride=true
    gravityScale=0.35
    simulationSpace=RootRelative
    alphaBoolBlend {blendOption=Linear}
    boundBone {boneName=dyn_earring_l_1}
    componentAppliedLinearAccClamp=<1000 1000 1000>
    componentLinearAccScale=<0.6 0.6 0.6>
    componentLinearVelScale=<0.5 0.5 0.5>
    gravityOverride {z=-980}
    physicsBodyDefinitions [{
      boundBone {boneName=dyn_earring_l_1}
      boxExtents=<15 15 15>
      constraintSetup {angularLimitsMax=<0 30 30>, angularLimitsMin=<0 -30 -30>, angularTarget=<1 0 0>}
      localJointOffset=<2 0 0>
     }]
    relativeSpaceBone {boneName=head}
   }
   n16 = Anim Dynamics - Bone: dyn_earring_l_2!(ComponentPose=@n15, Alpha=1, GravityOverride:Vector="0,0,-980", ExternalForce="0,0,0") -> (Pose) {
    angularDampingOverride=0.999
    angularSpringConstant=500
    bAngularSpring=true
    bOverrideAngularDamping=true
    bUseGravityOverride=true
    gravityScale=0.35
    simulationSpace=RootRelative
    alphaBoolBlend {blendOption=Linear}
    boundBone {boneName=dyn_earring_l_2}
    componentAppliedLinearAccClamp=<1000 1000 1000>
    componentLinearAccScale=<0.6 0.6 0.6>
    componentLinearVelScale=<0.5 0.5 0.5>
    gravityOverride {z=-980}
    physicsBodyDefinitions [{
      boundBone {boneName=dyn_earring_l_2}
      boxExtents=<10 10 10>
      constraintSetup {angularLimitsMax=<5 50 30>, angularLimitsMin=<-5 -30 -30>, angularTarget=<1 0 0>}
      localJointOffset=<2 0 0>
     }]
    relativeSpaceBone {boneName=head}
   }
   Save cached pose 'AnimBPPose'(Pose=Component To Local(ComponentPose=@n16)) {cacheName=AnimBPPose}
   n17 = Control Rig!(Source=Copy Pose From Mesh() {bUseAttachedParent=true}, Alpha, NoRB:bool=Get Without_RB(), CrouchMoving:double=1) -> (Pose) {
    alphaBoolBlend {blendOption=Linear}
    controlRigAssetReference {blueprintRigClass=&/F_MED_DollEnthusiast_FaceAcc_CtrlRig1.F_MED_DollEnthusiast_FaceAcc_CtrlRig1_C}
   }
   n18 = Control Rig(Source=Copy Pose From Mesh() {bUseAttachedParent=true}, NoRB=Get Without_RB(), CrouchMoving=0) {
    alphaBoolBlend {blendOption=Linear}
    controlRigAssetReference {blueprintRigClass=&F_MED_DollEnthusiast_FaceAcc_CtrlRig1_C}
   }
   n19 = Blend Poses by bool(BlendPose_0=@n17, BlendPose_1=@n18, BlendTime_0=0.5, BlendTime_1=0.5, bActiveValue=Get CrouchMoving() -> (CrouchMoving:bool)) {blendType=Linear}
   Save cached pose 'InputPose'(Pose=Component To Local(ComponentPose=Local To Component(LocalPose=@n19))) {cacheName=InputPose}
   n20 = "Constraint - dyn_hair_bk_l_3 (PP)"!(ComponentPose=Local To Component(LocalPose=Use cached pose 'AnimBPPose'()), Alpha=1, ConstraintWeights_0=0.3, ConstraintWeights_1:float=0.7) -> (Pose) {
    alphaBoolBlend {blendOption=Linear}
    boneToModify {boneName=dyn_hair_bk_l_3}
    constraintSetup [{targetBone {boneName=spine_05}}, {targetBone {boneName=clavicle_l}}]
    constraintWeights [1,1]
   }
   n21 = "Constraint - dyn_hair_bk_c_3 (PP)"!(ComponentPose=@n20, Alpha=1, ConstraintWeights_0=0.3, ConstraintWeights_1:float=0.7) -> (Pose) {
    alphaBoolBlend {blendOption=Linear}
    boneToModify {boneName=dyn_hair_bk_c_3}
    constraintSetup [{targetBone {boneName=spine_05}}, {targetBone {boneName=clavicle_r}}]
    constraintWeights [1,1]
   }
   n22 = "Constraint - dyn_hair_bk_r_3 (PP)"!(ComponentPose=@n21, Alpha=1, ConstraintWeights_0=0.3, ConstraintWeights_1:float=0.7) -> (Pose) {
    alphaBoolBlend {blendOption=Linear}
    boneToModify {boneName=dyn_hair_bk_r_3}
    constraintSetup [{targetBone {boneName=spine_05}}, {targetBone {boneName=clavicle_r}}]
    constraintWeights [1,1]
   }
   n23 = Control Rig!(Source=Component To Local(ComponentPose=@n22), Alpha) -> (Pose) {
    alphaBoolBlend {blendOption=Linear}
    controlRigAssetReference {blueprintRigClass=&/F_MED_DollEnthusiast_FaceAcc_Backpack_CtrlRig.F_MED_DollEnthusiast_FaceAcc_Backpack_CtrlRig_C}
   }
   Save cached pose 'WithoutRBPose'(Pose=Component To Local(ComponentPose=Local To Component(LocalPose=@n23))) {cacheName=WithoutRBPose}
   n24 = RigidBody!(
    ComponentPose=Local To Component(LocalPose=Use cached pose 'AnimBPPose'())
    Alpha
    OverrideWorldGravity:Vector="0,0,-980"
    ComponentLinearAccScale:Vector="0.3,0.3,0.3"
    ComponentLinearVelScale:Vector="0.2,0.2,0.2"
   ) -> (Pose) {
    bClampLinearTranslationLimitToRefPose=true
    bOverrideWorldGravity=true
    overridePhysicsAsset=&/F_MED_DollEnthusiast_FaceAcc_Physics1
    simulationSpace=BaseBoneSpace
    alphaBoolBlend {blendOption=Linear}
    alphaScaleBiasClamp {interpSpeedDecreasing=4, interpSpeedIncreasing=6}
    baseBoneRef {boneName=spine_05}
    componentAppliedLinearAccClamp=<3000 3000 3000>
    componentLinearAccScale=<0.4 0.4 0.4>
    componentLinearVelScale=<0.3 0.3 0.3>
    overrideWorldGravity {z=-875}
    simSpaceSettings {maxAngularAcceleration=200, maxAngularVelocity=5, maxLinearAcceleration=500, maxLinearVelocity=200}
   }
   n25 = RigidBody(
    ComponentPose=Local To Component(LocalPose=Use cached pose 'AnimBPPose'())
    OverrideWorldGravity="0,0,-980"
    ComponentLinearAccScale="0.3,0.3,0.3"
    ComponentLinearVelScale="0.2,0.2,0.2"
   ) {
    bClampLinearTranslationLimitToRefPose=true
    bOverrideWorldGravity=true
    overridePhysicsAsset=&/F_MED_DollEnthusiast_FaceAcc_Physics_InAir
    simulationSpace=BaseBoneSpace
    alphaBoolBlend {blendOption=Linear}
    alphaScaleBiasClamp {interpSpeedDecreasing=4, interpSpeedIncreasing=6}
    baseBoneRef {boneName=spine_05}
    componentAppliedLinearAccClamp=<3000 3000 3000>
    componentLinearAccScale=<0.4 0.4 0.4>
    componentLinearVelScale=<0.3 0.3 0.3>
    overrideWorldGravity {z=-875}
    simSpaceSettings {maxAngularAcceleration=200, maxAngularVelocity=5, maxLinearAcceleration=500, maxLinearVelocity=200}
   }
   Save cached pose 'RBPose'(Pose=Blend Poses by bool(BlendPose_0=Component To Local(ComponentPose=@n25), BlendPose_1=Component To Local(ComponentPose=@n24), bActiveValue=Get bIsSkyDiving()) {blendType=Linear}) {cacheName=RBPose}
  ]}
 EventGraph {nodes [
   Event BlueprintUpdateAnimation()
   > n1 = Sequence()
   n2 = Gravity Override State Function DEPRECATED(
    execute=@n1.then_1
    CharacterPartAnimInstance=Self-Reference()
    GameTimeInSeconds=GetGameTimeInSeconds()
    GravityOverrideParams=Get Gravity Override Params()
   )
   > n3 = Branch(Condition=Get SimpleJump() -> (SimpleJump:bool))
   > Set Override World Gravity(Override World Gravity:Vector="0,0,-650") -> (Output_Get:Vector)
   Set Override World Gravity(execute=@n3.else, Override World Gravity=@n2)
   n4 = CheckForBackbling()
   Set Wearing_Backpack(execute=@n1.then_0, Wearing_Backpack:bool=AND Boolean(A=@n4, B=NOT Boolean(A=@n4.bIsBackblingACape))) -> (Output_Get:bool)
   n5 = NOT Boolean(A=Get bIsSkyDiving())
   n6 = Get bIsJumping()
   n7 = Get bIsFalling()
   > Set SimpleJump(SimpleJump:bool=OR Boolean(A=AND Boolean(A=@n7, B=@n5), B=AND Boolean(A=@n5, B=@n6))) -> (Output_Get:bool)
   n8 = Get bRigidBodyEnabled()
   > Set Without_RB(Without_RB:bool=OR Boolean(A=NOT Boolean(A=@n8), B=Get Wearing_Backpack() -> (Wearing_Backpack:bool))) -> (Output_Get:bool)
   > Set CrouchMoving(CrouchMoving:bool=AND Boolean(A=Get bIsCrouching(), B=Get bIsMoving2D())) -> (Output_Get:bool)
  ]}
]
variables [
 CrouchMoving {type=boolean}
 Gravity Override Params {type=GravityOverrideParamsStruct}
 Override World Gravity {type=Vector}
 SimpleJump {type=boolean}
 Wearing_Backpack {type=boolean}
 Without_RB {type=boolean}
]
